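# Physics asset for the LatteStir back bling mesh, simulating a dangling chain attached to the pack's handle.
# A capsule on base_handle anchors six identical small capsules along the dyn_chain_a to dyn_chain_f bones, each jointed to the previous one.
# Twist is locked at the handle joint and tightly limited down the chain, while swing limits generally loosen toward the free end, so it sways without spinning.
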PhysicsAsset /BRCosmetics/Accessories/FORT_Backpacks/Backpack_M_MED_LatteStir/Meshes/M_MED_LatteStir_Pack_Physics
bodies {
 base_handle [{shapeType=Capsule, length=7.4574, radius=2.9264, center=<14.863 -0.0027 -2.1423>, rotation=<rot 0.0008 89.9834 -94.3881>}]
 dyn_chain_a [{shapeType=Capsule, length=2.2, radius=1.2, shapeName=dyn_chain_a_capsule, center=<2 0 0>, rotation=<rot -90 0 0>}]
 dyn_chain_b [{shapeType=Capsule, length=2.2, radius=1.2, shapeName=dyn_chain_b_capsule, center=<3.4 0 0>, rotation=<rot -90 0 0>}]
 dyn_chain_c [{shapeType=Capsule, length=2.2, radius=1.2, shapeName=dyn_chain_c_capsule, center=<3.4 0 0>, rotation=<rot -90 0 0>}]
 dyn_chain_d [{shapeType=Capsule, length=2.2, radius=1.2, shapeName=dyn_chain_d_capsule, center=<3.4 0 0>, rotation=<rot -90 0 0>}]
 dyn_chain_e [{shapeType=Capsule, length=2.2, radius=1.2, shapeName=dyn_chain_e_capsule, center=<3.4 0 0>, rotation=<rot -90 0 0>}]
 dyn_chain_f [{shapeType=Capsule, length=2.2, radius=1.2, shapeName=dyn_chain_f_capsule, center=<3.4 0 0>, rotation=<rot -90 0 0>}]
}
constraints [
 {bone1Name=dyn_chain_f, bone2Name=dyn_chain_e, swing1LimitDegrees=80, swing2LimitDegrees=75, twistLimitDegrees=2}
 {bone1Name=dyn_chain_a, bone2Name=base_handle, swing1LimitDegrees=25, swing2LimitDegrees=86, twistLimitDegrees=0, twistMotion=Locked}
 {bone1Name=dyn_chain_b, bone2Name=dyn_chain_a, swing1LimitDegrees=30, swing2LimitDegrees=45, twistLimitDegrees=2}
 {bone1Name=dyn_chain_c, bone2Name=dyn_chain_b, swing1LimitDegrees=60, swing2LimitDegrees=50, twistLimitDegrees=1}
 {bone1Name=dyn_chain_d, bone2Name=dyn_chain_c, swing1LimitDegrees=80, swing2LimitDegrees=45, twistLimitDegrees=2}
 {bone1Name=dyn_chain_e, bone2Name=dyn_chain_d, swing1LimitDegrees=75, swing2LimitDegrees=75, twistLimitDegrees=2}
]
