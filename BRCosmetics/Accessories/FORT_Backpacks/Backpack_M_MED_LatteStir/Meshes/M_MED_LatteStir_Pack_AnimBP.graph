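# Animation Blueprint for the M_MED_LatteStir backpack that adds physics-driven motion and posture correction to the pack.
# A RigidBody node feeds an additive pack-bone offset driven by the owner's spine yaw, cached as Dynamics.
# Fixed chain-bone offsets then branch on bIsInFrontEnd.
# The update event switches gravity and sim-space settings by skydive and parachute state, using upward gravity in freefall.
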AnimBlueprint /BRCosmetics/Accessories/FORT_Backpacks/Backpack_M_MED_LatteStir/Meshes/M_MED_LatteStir_Pack_AnimBP
target_skeleton=&/M_MED_LatteStir_Pack_Skeleton
graphs [
 AnimGraph {nodes [
   n1 = "Transform (Modify) Bone - Bone: base_chain_a"(
    ComponentPose=Local To Component(LocalPose=Use cached pose 'Dynamics'())
    Alpha=1
    Translation="-0.4,0.3,0.55"
    Rotation="7,10,25"
    Scale="1,1,1"
   ) {
    rotationMode=BMM_Additive
    rotationSpace=BCS_BoneSpace
    translationMode=BMM_Additive
    translationSpace=BCS_BoneSpace
    boneToModify {boneName=base_chain_a}
    rotation=<rot 7 10 25>
    translation=<-0.4 0.3 0.55>
   }
   n2 = "Transform (Modify) Bone - Bone: dyn_chain_b"(ComponentPose=@n1, Alpha=1, Translation="-0.2,0,0", Rotation="0,-2,0", Scale="1,1,1") {
    rotationMode=BMM_Additive
    rotationSpace=BCS_BoneSpace
    translationMode=BMM_Additive
    translationSpace=BCS_BoneSpace
    boneToModify {boneName=dyn_chain_b}
    rotation {yaw=-2}
    translation {x=-0.2}
   }
   n3 = "Transform (Modify) Bone - Bone: base_chain_a"!(ComponentPose=@n2, bAlphaBoolEnabled:bool=True, Translation="0,-0.2,-0.5", Rotation="0,-40,0", Scale="1,1,1") -> (Pose) {
    alphaInputType=Bool
    rotationMode=BMM_Additive
    rotationSpace=BCS_BoneSpace
    translationMode=BMM_Additive
    translationSpace=BCS_BoneSpace
    alphaBoolBlend {blendInTime=0.3, blendOutTime=0.3}
    boneToModify {boneName=base_chain_a}
    rotation {yaw=-40}
    translation {y=-0.2, z=-0.5}
   }
   n4 = "Transform (Modify) Bone - Bone: base_chain_a"!(
    ComponentPose=Local To Component(LocalPose=Use cached pose 'Dynamics'())
    Alpha:float=1
    Translation="-0.45,0.1,0.3"
    Rotation="-15,6,15"
    Scale="1,1,1"
   ) -> (Pose) {
    rotationMode=BMM_Additive
    rotationSpace=BCS_BoneSpace
    translationMode=BMM_Additive
    translationSpace=BCS_BoneSpace
    boneToModify {boneName=base_chain_a}
    rotation=<rot -15 6 15>
    translation=<-0.45 0.1 0.3>
   }
   n5 = "Transform (Modify) Bone - Bone: dyn_chain_b"(ComponentPose=@n4, Alpha=1, Translation="-0.1,0,0", Rotation="25,0,0", Scale="1,1,1") {
    rotationMode=BMM_Additive
    rotationSpace=BCS_BoneSpace
    translationMode=BMM_Additive
    translationSpace=BCS_BoneSpace
    boneToModify {boneName=dyn_chain_b}
    rotation {pitch=25}
    translation {x=-0.1}
   }
   Output Pose(Result=Blend Poses by bool(BlendPose_0=Component To Local(ComponentPose=@n5), BlendPose_1=Component To Local(ComponentPose=@n3), bActiveValue=Get bIsInFrontEnd())) {layerGroup=None}
   n6 = RigidBody!(
    ComponentPose=Local To Component(LocalPose=Input Pose())
    Alpha
    OverrideWorldGravity:Vector=Get gravity() -> (gravity:Vector)
    SimSpaceSettings:Sim Space Settings Structure=Get SimSpaceSettings() -> (SimSpaceSettings:Sim Space Settings Structure)
   ) -> (Pose) {
    bOverrideWorldGravity=true
    overridePhysicsAsset=&/M_MED_LatteStir_Pack_Physics
    overrideWorldGravity {z=-980}
   }
   n7 = Local To Component(LocalPose=Component To Local(ComponentPose=@n6))
   n8 = Get Backpack Translation() -> (Backpack Translation:Vector)
   n9 = Get Backpack Rotate() -> (Backpack Rotate:Rotator)
   Save cached pose 'Dynamics'(Pose=Component To Local(ComponentPose="Transform (Modify) Bone - Bone: pack"(ComponentPose=@n7, Translation=@n8, Rotation=@n9) {rotationMode=BMM_Additive, translationMode=BMM_Additive, boneToModify {boneName=pack}})) {cacheName=Dynamics}
  ]}
 EventGraph {nodes [
   Event BlueprintUpdateAnimation()
   > n1 = Sequence()
   > n2 = Branch(Condition=Get bIsSkyDiving())
   n3 = Get bIsParachuteOpened()
   > n4 = Branch(Condition=@n3)
   > Set gravity(gravity:Vector="0,0,-880") -> (Output_Get:Vector)
   > Set SimSpaceSettings(SimSpaceSettings:Sim Space Settings Structure=Get SimSpaceSettings_1() -> (SimSpaceSettings_1:Sim Space Settings Structure)) -> (Output_Get:Sim Space Settings Structure)
   > Set skydiveParachute(skydiveParachute:bool=false) -> (Output_Get:bool)
   Set gravity(execute=@n4.else, gravity="0,0,130")
   > Set SimSpaceSettings(SimSpaceSettings=Get SimSpaceSettings_0() -> (SimSpaceSettings_0:Sim Space Settings Structure))
   > Set skydiveParachute(skydiveParachute=true)
   Set gravity(execute=@n2.else, gravity="0,0,-1480")
   > Set SimSpaceSettings(SimSpaceSettings=Get SimSpaceSettings_1())
   n5 = GetOwnerSkeletalMeshComponent()
   n6 = BreakRotator(InRot=BreakTransform(InTransform=GetDeltaTransformFromRefPose(self=@n5, BoneName=spine_05, BaseName=spine_01)).Rotation)
   Set Backpack Rotate(execute=@n1.then_1, Backpack Rotate:Rotator=MakeRotator(Roll=MapRangeClamped(Value=@n6.Yaw, InRangeA=0, InRangeB=-35, OutRangeB=35))) -> (Output_Get:Rotator)
   > Set Backpack Translation(Backpack Translation:Vector=MakeVector(Y=MapRangeClamped(Value=@n6.Yaw, InRangeA=0, InRangeB=-30, OutRangeB=-3))) -> (Output_Get:Vector)
   TryGetPawnOwner()
   AND Boolean(A=NOT Boolean(A=@n3), B=false)
  ]}
]
variables [
 Backpack Rotate {type=Rotator}
 Backpack Translation {type=Vector}
 gravity {type=Vector}
 SimSpaceSettings {
  type=SimSpaceSettings
  default {
   dampingAlpha=1
   maxAngularAcceleration=10000
   maxAngularVelocity=10000
   maxLinearAcceleration=10000
   maxLinearVelocity=10000
   velocityScaleZ=1
   worldAlpha=0
   externalLinearDragV=<0 0 0>
  }
 }
 SimSpaceSettings_0 {
  type=SimSpaceSettings
  default {
   dampingAlpha=0.55
   maxAngularAcceleration=750
   maxAngularVelocity=20
   maxLinearAcceleration=225
   maxLinearVelocity=250
   velocityScaleZ=0.25
   worldAlpha=0.75
   externalLinearDragV=<0 0 0>
  }
 }
 SimSpaceSettings_1 {
  type=SimSpaceSettings
  default {
   dampingAlpha=0.12
   maxAngularAcceleration=37.32
   maxAngularVelocity=4.25
   maxLinearAcceleration=625
   maxLinearVelocity=200
   velocityScaleZ=0.12
   worldAlpha=0.9
   externalLinearDragV=<0 0 0>
  }
 }
 SimSpaceSettings_2 {
  type=SimSpaceSettings
  default {
   dampingAlpha=0.72
   maxAngularAcceleration=2500
   maxAngularVelocity=25
   maxLinearAcceleration=220
   maxLinearVelocity=250
   velocityScaleZ=0.3
   worldAlpha=0.85
   externalLinearDragV=<0.9 0.9 0.9>
  }
 }
 skydiveParachute {type=boolean}
]
